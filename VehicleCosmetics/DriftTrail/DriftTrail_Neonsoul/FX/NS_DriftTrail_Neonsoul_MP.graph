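# Niagara system for the Neonsoul drift trail vehicle cosmetic, drawing glow effects and trails at two wheels.
# Burst-spawned sprite/mesh cone and wheel-wrap particles re-sample skeletal sockets each update, with the wrap scaled per wheel.
# A ribbon emitter spawns two quality-scaled trails with backward velocity, inherited source movement and tapering width.
# All three emitters take per-variant painted colors and use vehicle-flag-driven tire FX masking, with a user offset shifting cone and ribbons.
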NiagaraSystem /VehicleCosmetics/DriftTrail/DriftTrail_Neonsoul/FX/NS_DriftTrail_Neonsoul_MP
dependencies {
 usedDataInterfaces [&NiagaraDataInterfaceCurve, &NiagaraDataInterfaceColorCurve, &NiagaraDataInterfaceSkeletalMesh]
 usedDynamicInputs []
 usedModules [
  &self:SystemSpawnScript.SystemScriptSource.SystemScriptGraph.NiagaraNodeAssignment_0.SetVariables_#1
  &FN_UnpackVehicleFlags
  /VehicleCosmetics/FX/Niagara/Modules/VC_SetVehicleSpeedAttribs
  &SystemState
  /Game/Effects/Niagara/Modules/Vehicle/FN_VehicleFlags_Cheatcodes
 ]
}
emitters {
 Cone_Complete_DrifTrail {
  emitterUpdateScript [
   &EmitterState(
    Inactive Response=::NewEnumerator1
    Life Cycle Mode=::NewEnumerator1
    Loop Behavior=::NewEnumerator1
    Loop Duration Mode=::NewEnumerator1
    UseLoopDelay=false
   )
   &SpawnRate(
    SpawnRate=/VehicleCosmetics/FX/Niagara/Modules/VC_Mask_Floats(
     FrontEnd Value:NiagaraFloat=200
     blsFrontEndPreview:NiagaraBool=false
     SpeedMin Value:NiagaraFloat=120
     SpeedMax Value:NiagaraFloat=180
     Speed Alpha:NiagaraFloat=$System.Speed_Alpha
    )
    Use Spawn Probability=false
   ) {enabled=false}
   &SpawnBurst_Instantaneous(Spawn Count=2, Use Spawn Probability=false, UseLoopCountLimit=false)
  ]
  particleSpawnScript [
   &VC_PaintedVariantsOverride(
    Black=<rgba 0.01 0.01 0.01 0.8>
    BurntSienna=<rgba 0.04 0.0073 0.0008 0.8>
    Cobalt=<rgba 0 0.0083 0.1 0.8>
    Crimson=<rgba 0.4 0 0 0.8>
    DefaultColor=<rgba 0.4 0 0 0.8>
    ForestGreen=<rgba 0 0.1 0 0.8>
    Grey=<rgba 0.05 0.05 0.05 0.8>
    Lime=<rgba 0.0583 0.1 0 0.8>
    Orange=<rgba 0.3 0.025 0 0.8>
    Pink=<rgba 0.2 0 0.1667 0.8>
    Purple=<rgba 0.025 0 0.1 0.8>
    R1 Black:LinearColor=<rgba 0.25 0.25 0.25 0.4>
    R1 BurntSienna:LinearColor=<rgba 2 0.5 0.2 0.4>
    R1 Cobalt:LinearColor=<rgba 0.2 0.5 2 0.8>
    R1 Crimson:LinearColor=<rgba 2 0.385 0.3 0.4>
    R1 DefaultColor=<rgba 2 0.29 0.2 0.4>
    R1 ForestGreen:LinearColor=<rgba 0.4 2 0.4 0.4>
    R1 Grey:LinearColor=<rgba 1 1 1 0.4>
    R1 Lime:LinearColor=<rgba 0.875 1.5 0 0.6>
    R1 Orange:LinearColor=<rgba 2 0.4 0 0.8>
    R1 Pink:LinearColor=<rgba 2 0.4 1.7333 0.8>
    R1 Purple:LinearColor=<rgba 0.9333 0.4 2 0.8>
    R1 Saffron:LinearColor=<rgba 2 1.5 0 0.4>
    R1 SkyBlue:LinearColor=<rgba 0.2 0.8 2 0.8>
    R1 TitaniumWhite:LinearColor=<rgba 2 2 2 0.4>
    R2 Black:LinearColor=<rgba 0 0 0 2>
    R2 BurntSienna:LinearColor=<rgba 0 0 0 2>
    R2 Cobalt:LinearColor=<rgba 0 0 0 2>
    R2 Crimson:LinearColor=<rgba 0 0 0 2>
    R2 DefaultColor:LinearColor=<rgba 0 0 0 2>
    R2 ForestGreen:LinearColor=<rgba 0 0 0 2>
    R2 Gold:LinearColor=<rgba 0 0 0 2>
    R2 Grey:LinearColor=<rgba 0 0 0 2>
    R2 Lime:LinearColor=<rgba 0 0 0 2>
    R2 Orange:LinearColor=<rgba 0 0 0 2>
    R2 Pink:LinearColor=<rgba 0 0 0 2>
    R2 Purple:LinearColor=<rgba 0 0 0 2>
    R2 Saffron:LinearColor=<rgba 0 0 0 2>
    R2 SkyBlue:LinearColor=<rgba 0 0 0 2>
    R2 TitaniumWhite:LinearColor=<rgba 0 0 0 2>
    Saffron=<rgba 0.1 0.075 0 0.8>
    SkyBlue=<rgba 0 0.0333 0.1 0.8>
    TitaniumWhite=<rgba 0.1 0.1 0.1 0.6>
    Use R1 Painted Overrides=true
    Use R2 Painted Overrides=true
   )
   &InitializeParticle(
    Color=$Particles.Color
    Mesh Scale=&Multiply_VectorByFloat(Vector=<0.2 0.45 0.2>, Float=1.8)
    Mesh Scale Mode=::NewEnumerator4
    Sprite Rotation Mode=::NewEnumerator2
    Sprite Size Mode=::NewEnumerator3
    Uniform Sprite Size=25
    UsePositionOffset=false
    Write Material Random=true
   )
   &PartitionParticles()
   &SkeletalMeshLocation(
    Bone / Socket Index=&Add_Integer(A=$Particles.PartitionParticles.Partition, B=4)
    Bone / Socket Sampling Mode=::NewEnumerator5
    Interpolate Mesh Sampling=false
    MaskSpawnGroup=false
    Mesh Sampling Type=::NewEnumerator6
    Sampled Position Offset=&AddVectorToPosition(Position=<-70 0 -20>, Vector=$User.offset)
    Skeletal Mesh=$System.SkeletalMesh
   )
   &AddVelocity(Velocity=&VC_Mask_Vectors(
     FrontEnd Value:Vector3f=<-1500 0 0>
     blsFrontEndPreview:NiagaraBool=false
     SpeedMin Value:Vector3f=<-800 0 0>
     SpeedMax Value:Vector3f=<-1 0 0>
     Speed Alpha:NiagaraFloat=$System.Speed_Alpha
    )) {enabled=false}
   &InitialMeshOrientation(Mesh Axis To Orient=<0 0 0>, Mesh Orientation Mode=::NewEnumerator2, Orientation Vector=<0 -1 0>, Use Rotation Vector=false)
   /SetVariables(
    Particles.colorglow:LinearColor=&MultiplyLinearColor(
     Linear Color=$Particles.R1 Color
     Scale Factor=&MakeVector4FromVectorAndFloat("Vector (RGB)"=&VectorFromFloat(Value=2), "Float (Alpha)"=1)
    )
    Particles.sizeglow:Vector2f=&Vector2DFromFloat(Value=75)
    Particles.spotcolor:LinearColor=&MultiplyLinearColor(
     Linear Color=$Particles.R1 Color
     Scale Factor=&MakeVector4FromVectorAndFloat("Vector (RGB)"=&VectorFromFloat(Value=0.5), "Float (Alpha)"=0.8)
    )
   ) {isSetParameters=true}
  ]
  particleUpdateScript [
   &FN_DisableTireFX(
    Mask Index Attribute:NiagaraInt32=$Particles.PartitionParticles.Partition
    Mask Index FL=2
    Mask Index FR=3
    Mask Index RL=0
    Mask Index RR=1
   )
   &ParticleState(Kill Particles When Lifetime Has Elapsed=false)
   &ScaleSpriteSize(
    Uniform Curve Index=$System.NormalizedLoopAge
    Uniform Curve Sprite Scale=DataInterface() {propertyValues {Curve {defaultValue=0.8, keys [<at 0 0.2 RCIM_Cubic>,<at 1 1 RCIM_Cubic>]}}}
   )
   &SolveForcesAndVelocity(Clamp Velocity=false, Limit Acceleration=false) {enabled=false}
   /SetVariables(Particles.ringcolor:LinearColor=&MultiplyLinearColor(
     Linear Color=$Particles.R1 Color
     Scale Factor=&MakeVector4FromVectorAndFloat(
      "Vector (RGB)"=&VectorFromFloat(Value=1)
      "Float (Alpha)"=&FloatFromCurve(
       FloatCurve=DataInterface() {propertyValues {Curve {
          defaultValue=1
          keys [
           {interpMode=RCIM_Cubic, tangentMode=RCTM_User, time=0, value=0}
           {interpMode=RCIM_Cubic, tangentMode=RCTM_User, time=0.5, value=1}
           {tangentMode=RCTM_User, time=1, value=0}
          ]
         }}}
       CurveIndex=$System.NormalizedLoopAge
      )
     )
    )) {isSetParameters=true}
   &SkeletalMeshLocation(
    Bone / Socket Index=&Add_Integer(A=$Particles.PartitionParticles.Partition, B=4)
    Bone / Socket Sampling Mode=::NewEnumerator5
    MaskSpawnGroup=false
    Mesh Sampling Type=::NewEnumerator6
    Sampled Position Offset=&AddVectorToPosition(Position=<-70 0 -20>, Vector=$User.offset)
    Skeletal Mesh=$System.SkeletalMesh
   ) {module=SkeletalMeshLocation001}
  ]
  rendererClasses [&NiagaraSpriteRendererProperties, &NiagaraMeshRendererProperties]
  renderers [
   {rendererClass=&NiagaraSpriteRendererProperties}
   {rendererClass=&NiagaraMeshRendererProperties, rendererIndex=1}
   {rendererClass=&NiagaraMeshRendererProperties, rendererIndex=2}
   {rendererClass=&NiagaraMeshRendererProperties, rendererIndex=3}
   {rendererClass=&NiagaraMeshRendererProperties, rendererIndex=4}
   {rendererClass=&NiagaraSpriteRendererProperties, rendererIndex=5}
  ]
 }
 NE_VC_DriftTrail_Ribbon_Parent {
  emitterUpdateScript [
   &EmitterState(Inactive Response=::NewEnumerator1, Life Cycle Mode=::NewEnumerator1, Loop Duration=0.1, UseLoopDelay=false)
   &VC_RateQualityHandler(Epic Spawn Rate=50, High Spawn Rate=40, Low Spawn Rate=20, Use Custom Rate?=true)
   &VC_SpawnRibbon_v1(Number of Ribbons=2, SpawnRate=$Emitter.Spawn Rate)
  ]
  particleSpawnScript [
   &VC_PaintedVariantsOverride() {enabled=false}
   &VC_InitializeRibbon_v1(
    Custom Offset=true
    Enable Skeletal Position=true
    Lifetime=1.2
    Local Vector Offset=&AddVectorToPosition(
     Position=/Niagara/DynamicInputs/TypeConversions/MakeCustomPositionFromBool(Bool:NiagaraBool=$User.bIsFrontendPreview, True Position:NiagaraPosition=<-10 0 -20>, False Position:NiagaraPosition=<-25 0 -20>)
     Vector=$User.offset
    )
    Offset Amount=1
    Ribbon Width=50
   )
   &AddVelocity(Velocity=<-1300 0 0>) {module=AddVelocity001}
  ]
  particleUpdateScript [
   &VC_PaintedVariantsOverride(
    Black=&ColorFromCurve(ColorCurve=DataInterface() {propertyValues {
       AlphaCurve {keys [
         {interpMode=RCIM_Cubic, leaveTangentWeight=0.3333, tangentMode=RCTM_User, time=0.02, value=0}
         <at 0.15 0.5>
         {arriveTangentWeight=0.3333, interpMode=RCIM_Cubic, tangentMode=RCTM_User, time=0.9, value=0}
        ]}
       BlueCurve {keys [{interpMode=RCIM_Cubic, leaveTangentWeight=0.3333, tangentMode=RCTM_User, time=0, value=0.2}, <at 0.8248 0.02>]}
       GreenCurve {keys [{interpMode=RCIM_Cubic, leaveTangentWeight=0.3333, tangentMode=RCTM_User, time=0, value=0.2}, <at 0.8248 0.02>]}
       RedCurve {keys [{interpMode=RCIM_Cubic, leaveTangentWeight=0.3333, tangentMode=RCTM_User, time=0, value=0.2}, <at 0.8248 0.02>]}
      }})
    BurntSienna=&ColorFromCurve(ColorCurve=DataInterface() {propertyValues {
       AlphaCurve {keys [
         {interpMode=RCIM_Cubic, leaveTangentWeight=0.3333, tangentMode=RCTM_User, time=0.02, value=0}
         <at 0.15 0.8>
         {arriveTangentWeight=0.3333, interpMode=RCIM_Cubic, tangentMode=RCTM_User, time=0.9, value=0}
        ]}
       BlueCurve {keys [{interpMode=RCIM_Cubic, leaveTangentWeight=0.3333, tangentMode=RCTM_User, time=0, value=0.055}, <at 0.5 0.03>]}
       GreenCurve {keys [{interpMode=RCIM_Cubic, leaveTangentWeight=0.3333, tangentMode=RCTM_User, time=0, value=0.2292}, <at 0.5 0.05>]}
       RedCurve {keys [{interpMode=RCIM_Cubic, leaveTangentWeight=0.3333, tangentMode=RCTM_User, time=0, value=1.1}, <at 0.5 0.15>]}
      }})
    Cobalt=&ColorFromCurve(ColorCurve=DataInterface() {propertyValues {
       AlphaCurve {keys [
         {interpMode=RCIM_Cubic, leaveTangentWeight=0.3333, tangentMode=RCTM_User, time=0.02, value=0}
         <at 0.15 0.9>
         {arriveTangentWeight=0.3333, interpMode=RCIM_Cubic, tangentMode=RCTM_User, time=0.9, value=0}
        ]}
       BlueCurve {keys [{interpMode=RCIM_Cubic, leaveTangentWeight=0.3333, tangentMode=RCTM_User, time=0, value=2}, <at 0.5 0.8>]}
       GreenCurve {keys [{interpMode=RCIM_Cubic, leaveTangentWeight=0.3333, tangentMode=RCTM_User, time=0, value=0.295}, <at 0.5 0.0683>]}
       RedCurve {keys [{interpMode=RCIM_Cubic, leaveTangentWeight=0.3333, tangentMode=RCTM_User, time=0, value=0.14}, <at 0.5 0.016>]}
      }})
    Crimson=&ColorFromCurve(ColorCurve=DataInterface() {propertyValues {
       AlphaCurve {keys [
         {interpMode=RCIM_Cubic, leaveTangentWeight=0.3333, tangentMode=RCTM_User, time=0.02, value=0}
         <at 0.15 0.8>
         {arriveTangentWeight=0.3333, interpMode=RCIM_Cubic, tangentMode=RCTM_User, time=0.9, value=0}
        ]}
       BlueCurve {keys [{interpMode=RCIM_Cubic, leaveTangentWeight=0.3333, tangentMode=RCTM_User, time=0, value=0.104}, <at 0.5 0.04>]}
       GreenCurve {keys [{interpMode=RCIM_Cubic, leaveTangentWeight=0.3333, tangentMode=RCTM_User, time=0, value=0.104}, <at 0.5 0.04>]}
       RedCurve {keys [{interpMode=RCIM_Cubic, leaveTangentWeight=0.3333, tangentMode=RCTM_User, time=0, value=1.3}, <at 0.5 0.5>]}
      }})
    DefaultColor=&ColorFromCurve(ColorCurve=DataInterface() {propertyValues {
       AlphaCurve {keys [
         {interpMode=RCIM_Cubic, leaveTangentWeight=0.3333, tangentMode=RCTM_User, time=0.02, value=0}
         <at 0.15 0.9>
         {arriveTangentWeight=0.3333, interpMode=RCIM_Cubic, tangentMode=RCTM_User, time=0.9, value=0}
        ]}
       BlueCurve {keys [{interpMode=RCIM_Cubic, leaveTangentWeight=0.3333, tangentMode=RCTM_User, time=0, value=0.026}, <at 0.5 0.3893>]}
       GreenCurve {keys [{interpMode=RCIM_Cubic, leaveTangentWeight=0.3333, tangentMode=RCTM_User, time=0, value=0.0472}, <at 0.5 0.0514>]}
       RedCurve {keys [{interpMode=RCIM_Cubic, leaveTangentWeight=0.3333, tangentMode=RCTM_User, time=0, value=1.3}, <at 0.5 0.4115>]}
      }})
    ForestGreen=&ColorFromCurve(ColorCurve=DataInterface() {propertyValues {
       AlphaCurve {keys [
         {interpMode=RCIM_Cubic, leaveTangentWeight=0.3333, tangentMode=RCTM_User, time=0.02, value=0}
         <at 0.15 0.9>
         {arriveTangentWeight=0.3333, interpMode=RCIM_Cubic, tangentMode=RCTM_User, time=0.9, value=0}
        ]}
       BlueCurve {keys [{interpMode=RCIM_Cubic, leaveTangentWeight=0.3333, tangentMode=RCTM_User, time=0, value=0.024}, <at 0.5 0.0032>]}
       GreenCurve {keys [{interpMode=RCIM_Cubic, leaveTangentWeight=0.3333, tangentMode=RCTM_User, time=0, value=0.6}, <at 0.5 0.16>]}
       RedCurve {keys [{interpMode=RCIM_Cubic, leaveTangentWeight=0.3333, tangentMode=RCTM_User, time=0, value=0.024}, <at 0.5 0.0032>]}
      }})
    Grey=&ColorFromCurve(ColorCurve=DataInterface() {propertyValues {
       AlphaCurve {keys [
         {interpMode=RCIM_Cubic, leaveTangentWeight=0.3333, tangentMode=RCTM_User, time=0.02, value=0}
         <at 0.15 0.8>
         {arriveTangentWeight=0.3333, interpMode=RCIM_Cubic, tangentMode=RCTM_User, time=0.9, value=0}
        ]}
       BlueCurve {keys [{interpMode=RCIM_Cubic, leaveTangentWeight=0.3333, tangentMode=RCTM_User, time=0, value=0.3}, <at 0.8223 0.04>]}
       GreenCurve {keys [{interpMode=RCIM_Cubic, leaveTangentWeight=0.3333, tangentMode=RCTM_User, time=0, value=0.3}, <at 0.8223 0.04>]}
       RedCurve {keys [{interpMode=RCIM_Cubic, leaveTangentWeight=0.3333, tangentMode=RCTM_User, time=0, value=0.3}, <at 0.8223 0.04>]}
      }})
    Lime=&ColorFromCurve(ColorCurve=DataInterface() {propertyValues {
       AlphaCurve {keys [
         {interpMode=RCIM_Cubic, leaveTangentWeight=0.3333, tangentMode=RCTM_User, time=0.02, value=0}
         <at 0.15 0.9>
         {arriveTangentWeight=0.3333, interpMode=RCIM_Cubic, tangentMode=RCTM_User, time=0.9, value=0}
        ]}
       BlueCurve {keys [{interpMode=RCIM_Cubic, leaveTangentWeight=0.3333, tangentMode=RCTM_User, time=0, value=0.016}, <at 0.5 0.004>]}
       GreenCurve {keys [{interpMode=RCIM_Cubic, leaveTangentWeight=0.3333, tangentMode=RCTM_User, time=0, value=0.8}, <at 0.5 0.2>]}
       RedCurve {keys [{interpMode=RCIM_Cubic, leaveTangentWeight=0.3333, tangentMode=RCTM_User, time=0, value=0.4733}, <at 0.5 0.1183>]}
      }})
    Orange=&ColorFromCurve(ColorCurve=DataInterface() {propertyValues {
       AlphaCurve {keys [
         {interpMode=RCIM_Cubic, leaveTangentWeight=0.3333, tangentMode=RCTM_User, time=0.02, value=0}
         <at 0.15 0.8>
         {arriveTangentWeight=0.3333, interpMode=RCIM_Cubic, tangentMode=RCTM_User, time=0.9, value=0}
        ]}
       BlueCurve {keys [{interpMode=RCIM_Cubic, leaveTangentWeight=0.3333, tangentMode=RCTM_User, time=0, value=0.015}, <at 0.5 0.024>]}
       GreenCurve {keys [{interpMode=RCIM_Cubic, leaveTangentWeight=0.3333, tangentMode=RCTM_User, time=0, value=0.2625}, <at 0.5 0.1008>]}
       RedCurve {keys [{interpMode=RCIM_Cubic, leaveTangentWeight=0.3333, tangentMode=RCTM_User, time=0, value=1.5}, <at 0.5 0.6>]}
      }})
    Pink=&ColorFromCurve(ColorCurve=DataInterface() {propertyValues {
       AlphaCurve {keys [
         {interpMode=RCIM_Cubic, leaveTangentWeight=0.3333, tangentMode=RCTM_User, time=0.02, value=0}
         <at 0.15 0.8>
         {arriveTangentWeight=0.3333, interpMode=RCIM_Cubic, tangentMode=RCTM_User, time=0.9, value=0}
        ]}
       BlueCurve {keys [{interpMode=RCIM_Cubic, leaveTangentWeight=0.3333, tangentMode=RCTM_User, time=0, value=1.2133}, <at 0.5 0.6933>]}
       GreenCurve {keys [{interpMode=RCIM_Cubic, leaveTangentWeight=0.3333, tangentMode=RCTM_User, time=0, value=0.28}, <at 0.5 0.16>]}
       RedCurve {keys [{interpMode=RCIM_Cubic, leaveTangentWeight=0.3333, tangentMode=RCTM_User, time=0, value=1.4}, <at 0.5 0.8>]}
      }})
    Purple=&ColorFromCurve(ColorCurve=DataInterface() {propertyValues {
       AlphaCurve {keys [
         {interpMode=RCIM_Cubic, leaveTangentWeight=0.3333, tangentMode=RCTM_User, time=0.02, value=0}
         <at 0.15 0.8>
         {arriveTangentWeight=0.3333, interpMode=RCIM_Cubic, tangentMode=RCTM_User, time=0.9, value=0}
        ]}
       BlueCurve {keys [{interpMode=RCIM_Cubic, leaveTangentWeight=0.3333, tangentMode=RCTM_User, time=0, value=2}, <at 0.5 0.6>]}
       GreenCurve {keys [{interpMode=RCIM_Cubic, leaveTangentWeight=0.3333, tangentMode=RCTM_User, time=0, value=0.3}, <at 0.5 0.03>]}
       RedCurve {keys [{interpMode=RCIM_Cubic, leaveTangentWeight=0.3333, tangentMode=RCTM_User, time=0, value=0.6967}, <at 0.5 0.087>]}
      }})
    R1 Black:LinearColor=&ColorFromCurve(ColorCurve=DataInterface() {propertyValues {
       AlphaCurve {keys [<at 0.02 0>, <at 0.15 2>, {arriveTangentWeight=0.3333, interpMode=RCIM_Cubic, tangentMode=RCTM_User, time=0.8, value=0}]}
       BlueCurve {keys [<at 0 0>]}
       GreenCurve {keys [<at 0 0>]}
       RedCurve {keys [<at 0 0>]}
      }})
    R1 BurntSienna:LinearColor=&ColorFromCurve(ColorCurve=DataInterface() {propertyValues {
       AlphaCurve {keys [
         {interpMode=RCIM_Cubic, leaveTangentWeight=0.3333, tangentMode=RCTM_User, time=0.02, value=0}
         <at 0.15 2>
         {arriveTangentWeight=0.3333, interpMode=RCIM_Cubic, tangentMode=RCTM_User, time=0.8, value=0}
        ]}
       BlueCurve {keys [<at 0 0>]}
       GreenCurve {keys [<at 0 0>]}
       RedCurve {keys [<at 0 0>]}
      }})
    R1 Cobalt:LinearColor=&ColorFromCurve(ColorCurve=DataInterface() {propertyValues {
       AlphaCurve {keys [<at 0.02 0>, <at 0.15 2>, {arriveTangentWeight=0.3333, interpMode=RCIM_Cubic, tangentMode=RCTM_User, time=0.8, value=0}]}
       BlueCurve {keys [<at 0 0>]}
       GreenCurve {keys [<at 0 0>]}
       RedCurve {keys [<at 0 0>]}
      }})
    R1 Crimson:LinearColor=&ColorFromCurve(ColorCurve=DataInterface() {propertyValues {
       AlphaCurve {keys [<at 0.02 0>, <at 0.15 2>, {arriveTangentWeight=0.3333, interpMode=RCIM_Cubic, tangentMode=RCTM_User, time=0.8, value=0}]}
       BlueCurve {keys [<at 0 0>]}
       GreenCurve {keys [<at 0 0>]}
       RedCurve {keys [<at 0 0>]}
      }})
    R1 DefaultColor=&ColorFromCurve(ColorCurve=DataInterface() {propertyValues {
       AlphaCurve {keys [<at 0.02 0>, <at 0.15 2>, {arriveTangentWeight=0.3333, interpMode=RCIM_Cubic, tangentMode=RCTM_User, time=0.8, value=0}]}
       BlueCurve {keys [<at 0 0>]}
       GreenCurve {keys [<at 0 0>]}
       RedCurve {keys [<at 0 0>]}
      }})
    R1 ForestGreen:LinearColor=&ColorFromCurve(ColorCurve=DataInterface() {propertyValues {
       AlphaCurve {keys [<at 0.02 0>, <at 0.15 2>, {arriveTangentWeight=0.3333, interpMode=RCIM_Cubic, tangentMode=RCTM_User, time=0.8, value=0}]}
       BlueCurve {keys [<at 0 0>]}
       GreenCurve {keys [<at 0 0>]}
       RedCurve {keys [<at 0 0>]}
      }})
    R1 Gold:LinearColor=&ColorFromCurve(ColorCurve=DataInterface() {propertyValues {
       AlphaCurve {keys [<at 0.02 0>, <at 0.15 2>, {arriveTangentWeight=0.3333, interpMode=RCIM_Cubic, tangentMode=RCTM_User, time=0.8, value=0}]}
       BlueCurve {keys [<at 0 0>]}
       GreenCurve {keys [<at 0 0>]}
       RedCurve {keys [<at 0 0>]}
      }})
    R1 Grey:LinearColor=&ColorFromCurve(ColorCurve=DataInterface() {propertyValues {
       AlphaCurve {keys [<at 0.02 0>, <at 0.15 2>, {arriveTangentWeight=0.3333, interpMode=RCIM_Cubic, tangentMode=RCTM_User, time=0.8, value=0}]}
       BlueCurve {keys [<at 0 0>]}
       GreenCurve {keys [<at 0 0>]}
       RedCurve {keys [<at 0 0>]}
      }})
    R1 Lime:LinearColor=&ColorFromCurve(ColorCurve=DataInterface() {propertyValues {
       AlphaCurve {keys [<at 0.02 0>, <at 0.15 2>, {arriveTangentWeight=0.3333, interpMode=RCIM_Cubic, tangentMode=RCTM_User, time=0.8, value=0}]}
       BlueCurve {keys [<at 0 0>]}
       GreenCurve {keys [<at 0 0>]}
       RedCurve {keys [<at 0 0>]}
      }})
    R1 Orange:LinearColor=&ColorFromCurve(ColorCurve=DataInterface() {propertyValues {
       AlphaCurve {keys [<at 0.02 0>, <at 0.15 2>, {arriveTangentWeight=0.3333, interpMode=RCIM_Cubic, tangentMode=RCTM_User, time=0.8, value=0}]}
       BlueCurve {keys [<at 0 0>]}
       GreenCurve {keys [<at 0 0>]}
       RedCurve {keys [<at 0 0>]}
      }})
    R1 Pink:LinearColor=&ColorFromCurve(ColorCurve=DataInterface() {propertyValues {
       AlphaCurve {keys [<at 0.02 0>, <at 0.15 2>, {arriveTangentWeight=0.3333, interpMode=RCIM_Cubic, tangentMode=RCTM_User, time=0.8, value=0}]}
       BlueCurve {keys [<at 0 0>]}
       GreenCurve {keys [<at 0 0>]}
       RedCurve {keys [<at 0 0>]}
      }})
    R1 Purple:LinearColor=&ColorFromCurve(ColorCurve=DataInterface() {propertyValues {
       AlphaCurve {keys [<at 0.02 0>, <at 0.15 2>, {arriveTangentWeight=0.3333, interpMode=RCIM_Cubic, tangentMode=RCTM_User, time=0.8, value=0}]}
       BlueCurve {keys [<at 0 0>]}
       GreenCurve {keys [<at 0 0>]}
       RedCurve {keys [<at 0 0>]}
      }})
    R1 Saffron:LinearColor=&ColorFromCurve(ColorCurve=DataInterface() {propertyValues {
       AlphaCurve {keys [<at 0.02 0>, <at 0.15 2>, {arriveTangentWeight=0.3333, interpMode=RCIM_Cubic, tangentMode=RCTM_User, time=0.8, value=0}]}
       BlueCurve {keys [<at 0 0>]}
       GreenCurve {keys [<at 0 0>]}
       RedCurve {keys [<at 0 0>]}
      }})
    R1 SkyBlue:LinearColor=&ColorFromCurve(ColorCurve=DataInterface() {propertyValues {
       AlphaCurve {keys [<at 0.02 0>, <at 0.15 2>, {arriveTangentWeight=0.3333, interpMode=RCIM_Cubic, tangentMode=RCTM_User, time=0.8, value=0}]}
       BlueCurve {keys [<at 0 0>]}
       GreenCurve {keys [<at 0 0>]}
       RedCurve {keys [<at 0 0>]}
      }})
    R1 TitaniumWhite:LinearColor=&ColorFromCurve(ColorCurve=DataInterface() {propertyValues {
       AlphaCurve {keys [<at 0.02 0>, <at 0.15 2>, {arriveTangentWeight=0.3333, interpMode=RCIM_Cubic, tangentMode=RCTM_User, time=0.8, value=0}]}
       BlueCurve {keys [<at 0 0>]}
       GreenCurve {keys [<at 0 0>]}
       RedCurve {keys [<at 0 0>]}
      }})
    Saffron=&ColorFromCurve(ColorCurve=DataInterface() {propertyValues {
       AlphaCurve {keys [
         {interpMode=RCIM_Cubic, leaveTangentWeight=0.3333, tangentMode=RCTM_User, time=0.02, value=0}
         <at 0.15 0.8>
         {arriveTangentWeight=0.3333, interpMode=RCIM_Cubic, tangentMode=RCTM_User, time=0.9, value=0}
        ]}
       BlueCurve {keys [{interpMode=RCIM_Cubic, leaveTangentWeight=0.3333, tangentMode=RCTM_User, time=0, value=0.072}, <at 0.5 0.03>]}
       GreenCurve {keys [{interpMode=RCIM_Cubic, leaveTangentWeight=0.3333, tangentMode=RCTM_User, time=0, value=0.693}, <at 0.5 0.41>]}
       RedCurve {keys [{interpMode=RCIM_Cubic, leaveTangentWeight=0.3333, tangentMode=RCTM_User, time=0, value=0.9}, <at 0.5 0.6>]}
      }})
    SkyBlue=&ColorFromCurve(ColorCurve=DataInterface() {propertyValues {
       AlphaCurve {keys [
         {interpMode=RCIM_Cubic, leaveTangentWeight=0.3333, tangentMode=RCTM_User, time=0.02, value=0}
         <at 0.15 0.9>
         {arriveTangentWeight=0.3333, interpMode=RCIM_Cubic, tangentMode=RCTM_User, time=0.9, value=0}
        ]}
       BlueCurve {keys [{interpMode=RCIM_Cubic, leaveTangentWeight=0.3333, tangentMode=RCTM_User, time=0, value=1.3}, <at 0.5 0.8>]}
       GreenCurve {keys [{interpMode=RCIM_Cubic, leaveTangentWeight=0.3333, tangentMode=RCTM_User, time=0, value=0.4972}, <at 0.5 0.2773>]}
       RedCurve {keys [{interpMode=RCIM_Cubic, leaveTangentWeight=0.3333, tangentMode=RCTM_User, time=0, value=0.065}, <at 0.5 0.016>]}
      }})
    TitaniumWhite=&ColorFromCurve(ColorCurve=DataInterface() {propertyValues {
       AlphaCurve {keys [
         {interpMode=RCIM_Cubic, leaveTangentWeight=0.3333, tangentMode=RCTM_User, time=0.02, value=0}
         <at 0.15 0.8>
         {arriveTangentWeight=0.3333, interpMode=RCIM_Cubic, tangentMode=RCTM_User, time=0.9, value=0}
        ]}
       BlueCurve {keys [{interpMode=RCIM_Cubic, leaveTangentWeight=0.3333, tangentMode=RCTM_User, time=0, value=0.85}, <at 0.5512 0.3>]}
       GreenCurve {keys [{interpMode=RCIM_Cubic, leaveTangentWeight=0.3333, tangentMode=RCTM_User, time=0, value=0.85}, <at 0.5512 0.3>]}
       RedCurve {keys [{interpMode=RCIM_Cubic, leaveTangentWeight=0.3333, tangentMode=RCTM_User, time=0, value=0.85}, <at 0.5512 0.3>]}
      }})
    Use R1 Painted Overrides=true
   ) {module=VC_PaintedVariantsOverride001}
   &FN_DisableTireFX(Mask Index Attribute:NiagaraInt32=$Particles.SpawnGroup, Mask Index FL=2, Mask Index FR=3, Mask Index RL=0, Mask Index RR=1)
   &ParticleState()
   &InheritSourceMovement(
    Applied Position Scale=&Multiply_VectorByFloat(
     Vector=<1 1 1>
     Float=&FloatFromCurve(FloatCurve=DataInterface() {propertyValues {Curve {keys [
          {interpMode=RCIM_Cubic, leaveTangent=-0.0051, tangentMode=RCTM_Break, time=0, value=1}
          {arriveTangent=-0.247, interpMode=RCIM_Cubic, tangentMode=RCTM_Break, time=1, value=0.9}
         ]}}})
    )
    ApplyForce=false
    ApplyPosition=true
    ApplyPositionFalloff=false
   )
   &SolveForcesAndVelocity(Clamp Velocity=false, Limit Acceleration=false)
   &VC_UpdateRibbon_v1(Enable Skeletal Position=true)
   &ScaleRibbonWidth(Ribbon Width Scale=&FloatFromCurve(FloatCurve=DataInterface() {propertyValues {Curve {keys [
         {arriveTangent=0.3849, interpMode=RCIM_Cubic, leaveTangent=0.3849, leaveTangentWeight=0.3333, tangentMode=RCTM_User, time=0, value=1}
         {arriveTangent=-0.0697, arriveTangentWeight=0.3333, interpMode=RCIM_Cubic, leaveTangent=-0.0697, tangentMode=RCTM_User, time=1, value=0.6}
        ]}}}))
   /SetVariables(Particles.width2:NiagaraFloat=&Multiply_Float(
     A=&FloatFromCurve(FloatCurve=DataInterface() {propertyValues {Curve {keys [
          {arriveTangent=0.3849, interpMode=RCIM_Cubic, leaveTangent=0.3849, leaveTangentWeight=0.3333, tangentMode=RCTM_User, time=0, value=1}
          {arriveTangent=-0.0697, arriveTangentWeight=0.3333, interpMode=RCIM_Cubic, leaveTangent=-0.0697, tangentMode=RCTM_User, time=1, value=0.6}
         ]}}})
     B=&SelectFloatFromArray(
      Direct Array Index=$User.ColorVariantID
      Float Selection Array=DataInterface() {propertyValues {FloatData [55,55,60,60,55,55,56,60,60,60,56,55,60,56,56]}}
     )
    )) {isSetParameters=true}
  ]
  rendererClasses [&NiagaraRibbonRendererProperties]
  renderers [
   {rendererClass=&NiagaraRibbonRendererProperties}
   {rendererClass=&NiagaraRibbonRendererProperties, rendererIndex=1}
  ]
 }
 WheelWrap_NeonsoulDT001 {
  emitterUpdateScript [
   &EmitterState(
    Inactive Response=::NewEnumerator1
    Life Cycle Mode=::NewEnumerator1
    Loop Behavior=::NewEnumerator1
    Loop Duration Mode=::NewEnumerator1
    UseLoopDelay=false
   )
   &SpawnBurst_Instantaneous(Spawn Count=2, Use Spawn Probability=false, UseLoopCountLimit=false)
  ]
  particleSpawnScript [
   &VC_PaintedVariantsOverride(
    Black=<rgba 0.03 0.03 0.03 2>
    BurntSienna=<rgba 0.2 0.0417 0.01 1>
    Cobalt=<rgba 0.015 0.0625 0.3 1>
    Crimson=<rgba 0.3 0.015 0.015 1>
    DefaultColor=<rgba 0.3 0.015 0.0625 1>
    ForestGreen=<rgba 0.0055 0.11 0.0055 2>
    Grey=<rgba 0.05 0.05 0.05 2>
    Lime=<rgba 0.0785 0.13 0.0065 2>
    Orange=<rgba 0.22 0.0458 0.011 2>
    Pink=<rgba 0.3 0.06 0.26 2>
    Purple=<rgba 0.0625 0.025 0.25 2>
    Saffron=<rgba 0.2 0.1525 0.01 2>
    SkyBlue=<rgba 0.0125 0.1312 0.25 2>
    TitaniumWhite=<rgba 0.15 0.15 0.15 2>
   )
   &InitializeParticle(Color=$Particles.Color, Mesh Scale Mode=::NewEnumerator4, UsePositionOffset=false, Write Material Random=true)
   &PartitionParticles()
   &SkeletalMeshLocation(
    Bone / Socket Index=&Add_Integer(A=$Particles.PartitionParticles.Partition, B=4)
    Bone / Socket Sampling Mode=::NewEnumerator5
    Interpolate Mesh Sampling=false
    MaskSpawnGroup=false
    Mesh Sampling Type=::NewEnumerator6
    Skeletal Mesh=$System.SkeletalMesh
   )
  ]
  particleUpdateScript [
   &FN_DisableTireFX(
    Mask Index Attribute:NiagaraInt32=$Particles.PartitionParticles.Partition
    Mask Index FL=2
    Mask Index FR=3
    Mask Index RL=0
    Mask Index RR=1
   )
   &VC_WheelWrapScale(Wheel Index Attribute:NiagaraInt32=&Add_Integer(A=$Particles.PartitionParticles.Partition, B=2))
   &ParticleState(Kill Particles When Lifetime Has Elapsed=false)
   &UpdateMeshOrientation(Rotation Coordinate Space=::NewEnumerator2, Rotation Vector=<0 1 0>) {enabled=false}
   &SkeletalMeshLocation(
    Bone / Socket Index=&Add_Integer(A=$Particles.PartitionParticles.Partition, B=4)
    Bone / Socket Sampling Mode=::NewEnumerator5
    MaskSpawnGroup=false
    Mesh Sampling Type=::NewEnumerator6
    Skeletal Mesh=$System.SkeletalMesh
   ) {module=SkeletalMeshLocation001}
   &SolveForcesAndVelocity(Clamp Velocity=false, Limit Acceleration=false)
  ]
  rendererClasses [&NiagaraMeshRendererProperties]
  renderers [{rendererClass=&NiagaraMeshRendererProperties, rendererIndex=0}]
 }
}
summary {
 systemName=NS_DriftTrail_Neonsoul_MP
 emitters [
  {
   emitterName=Cone_Complete_DrifTrail
   rendererClasses [&NiagaraSpriteRendererProperties, &NiagaraMeshRendererProperties]
  }
  {emitterName=NE_VC_DriftTrail_Ribbon_Parent, rendererClasses [&NiagaraRibbonRendererProperties]}
  {emitterName=WheelWrap_NeonsoulDT001, rendererClasses [&NiagaraMeshRendererProperties]}
 ]
 userVariables [
  User.bIsFrontendPreview {type=NiagaraBool}
  User.bIsPainted {type=NiagaraBool}
  User.ColorVariantID {type=NiagaraInt32}
  User.offset {type=Vector3f, description="offset from wheels(set through cosmetic template)", defaultValue {struct=Vector3f, value=<0 0 0>}}
  User.PaintedColor {type=LinearColor, description="", defaultValue {struct=LinearColor, value=<rgba 1 1 1 1>}}
  User.PaintedScalar {type=NiagaraFloat, defaultValue=1, description=""}
  User.VehicleFlags {type=NiagaraInt32, defaultValue=806289408}
  User.Wheel_Radius_Array {
   type=NiagaraDataInterfaceArrayFloat
   defaultValue {
    struct=NiagaraExt_VariableValue_DataInterface
    value {
     dataInterface=&self:NiagaraDataInterfaceArrayFloat_1
     dataInterfaceClass=&NiagaraDataInterfaceArrayFloat
    }
   }
  }
  User.Wheel_Width_Array {
   type=NiagaraDataInterfaceArrayFloat
   defaultValue {
    struct=NiagaraExt_VariableValue_DataInterface
    value {
     dataInterface=&self:NiagaraDataInterfaceArrayFloat_0
     dataInterfaceClass=&NiagaraDataInterfaceArrayFloat
    }
   }
  }
  User.WheelTouchingArray {
   type=NiagaraDataInterfaceArrayFloat
   defaultValue {
    struct=NiagaraExt_VariableValue_DataInterface
    value {
     dataInterface=&self:NiagaraDataInterfaceArrayFloat_2
     dataInterfaceClass=&NiagaraDataInterfaceArrayFloat
    }
   }
  }
 ]
}
